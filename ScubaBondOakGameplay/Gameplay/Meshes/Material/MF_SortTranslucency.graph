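# Material function that fixes translucency sorting by hiding translucent pixels lying behind the surface recorded in Custom Depth.
# It transforms the camera-to-pixel vector into view space and takes Z as pixel depth, scaled by just under one as a self-occlusion bias.
# An If compares that depth to the CustomDepth scene texture, outputting zero when farther and the Opacity input when nearer.
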MaterialFunction /ScubaBondOakGameplay/Gameplay/Meshes/Material/MF_SortTranslucency
graph {
 properties {previewBlendMode=BLEND_Translucent}
 nodes [
  n1 = WorldPosition()
  n2 = CameraPositionWS()
  n3 = ComponentMask(None=Transform(None=Subtract(A=@n1.XYZ, B=@n2)) {transformSourceType=TRANSFORMSOURCE_World, transformType=TRANSFORM_View}) {b=true}
  n4 = SceneTexture() {sceneTextureId=PPI_CustomDepth}
  FunctionOutput(None=If(
    A=Multiply(A=@n3) {constB=0.9999}
    B=ComponentMask(None=@n4.Color) {r=true}
    "A > B"=Constant()
    "A < B"=FunctionInput() {inputName=Opacity, inputType=FunctionInput_Scalar}
   ))
 ]
}
