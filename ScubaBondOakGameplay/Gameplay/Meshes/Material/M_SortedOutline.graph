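# M_SortedOutline is an unlit additive material for sorted outlines on skeletal, mesh-deformed and instanced static meshes.
# An Opacity parameter feeds the MF_SortTranslucency function, whose name points to translucency ordering.
# An EmissiveColor parameter, black by default, is multiplied by itself to square the color.
# The material allows translucent custom depth writes and casts no ray-traced shadows.
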Material /ScubaBondOakGameplay/Gameplay/Meshes/Material/M_SortedOutline
graph {
 properties {
  allowTranslucentCustomDepthWrites=true
  bAutomaticallySetUsageInEditor=false
  bCastRayTracedShadows=false
  bUsedWithInstancedStaticMeshes=true
  bUsedWithMeshDeformer=true
  bUsedWithSkeletalMesh=true
  blendMode=BLEND_Additive
  opacityMaskClipValue=0.001
  pixelDepthOffsetMode=PDOM_Legacy
  shadingModel=MSM_Unlit
  translucencyLightingMode=TLM_Surface
  usedShadingModels=Unlit
 }
 nodes [
  MaterialFunctionCall!(Opacity=ScalarParameter() {defaultValue=1, parameterName=Opacity, sliderMax=1, sliderMin=0.001}) {materialFunction=&/MF_SortTranslucency}
  n1 = VectorParameter() {parameterName=EmissiveColor, defaultValue=<rgba 0 0 0 1>}
  Multiply(A=@n1.RGB, B=@n1.RGB)
 ]
}
